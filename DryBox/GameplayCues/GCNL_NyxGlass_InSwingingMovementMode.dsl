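; Looping gameplay cue that drives swinging audio and the sound indicator while a player is in the NyxGlass weapon's swinging movement mode.
; On start it attaches layered audio to the pawn, flags local/team, and binds actor hits for collision sounds while attached.
; Weapon state changes set IsMaxVelocity and start a repeating timer easing SpeedMultiplier toward a velocity-mapped value until close.
; Removal fades audio, unbinds events and resets parameters.
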(event EventOnLoopingStart(EitherVFX-preferthisversion)
  (params (MyTarget :type Actor%20Object%20Reference) (Parameters :type Gameplay%20Cue%20Parameters%20Structure%20%28by%20ref%29) (ParticleComponents :type Array%20of%20FXSystem%20Component%20Object%20References) (AudioComponents :type Array%20of%20Audio%20Component%20Object%20References))
  (CallFunction|AudioInit MyTarget)
  (bind _asplayer_pawn_athena (Utilities|Casting|CastToPlayerPawn_Athena MyTarget))
  (SoundIndicator|StartSoundIndicatorLooping (Variables|/Script/FortniteGame.FortWeapon|GetSoundIndicatorComponent (Weapon|GetCurrentWeapon _asplayer_pawn_athena)) (Transformation|GetActorLocation _asplayer_pawn_athena) (Variables|/Script/Engine.AudioComponent|GetSound (Variables|FortGameplayCueNotifyLoop_Swinging|GetLayeredAudioComp)) 1.0 0)
  (bind _asb_weapon_nyx_glass (Utilities|Casting|CastToB_Weapon_NyxGlass (Weapon|GetCurrentWeapon _asplayer_pawn_athena)))
  (bind _output_get (Variables|Default|SetNyxGlassReference _asb_weapon_nyx_glass))
  (Default|BindEventtoOnStateChanged _output_get (EventDispatchers|CreateEvent :!event Custom|OnNyxGlassStateChanged)))

(event EventOnRemoval(EitherVFX-preferthisversion)
  (params (MyTarget :type Actor%20Object%20Reference) (Parameters :type Gameplay%20Cue%20Parameters%20Structure%20%28by%20ref%29) (ParticleComponents :type Array%20of%20FXSystem%20Component%20Object%20References) (AudioComponents :type Array%20of%20Audio%20Component%20Object%20References) (BurstCameraShakeInstance :type Camera%20Shake%20Base%20Object%20Reference) (BurstDecalInstance :type Decal%20Actor%20Object%20Reference))
  (Default|AudioUninit)
  (bind _asplayer_pawn_athena (Utilities|Casting|CastToPlayerPawn_Athena MyTarget))
  (SoundIndicator|StopSoundIndicatorLooping (Variables|/Script/FortniteGame.FortWeapon|GetSoundIndicatorComponent (Weapon|GetCurrentWeapon _asplayer_pawn_athena)))
  (Variables|Default|GetNyxGlassReference
    (:then
      (bind _nyxglassreference (Variables|Default|GetNyxGlassReference)
        (:then
          (Default|UnbindEventfromOnStateChanged _nyxglassreference (EventDispatchers|CreateEvent :!event Custom|OnNyxGlassStateChanged))
          (Variables|Default|SetSpeedMultiplier 0.8))
        (:else)))
    (:else)))

(event Custom|OnNyxGlassStateChanged
  (Variables|FortGameplayCueNotifyLoop_Swinging|GetPlayerPawn
    (:then
      (bind _nyxglassreference (Variables|Default|GetNyxGlassReference)
        (:then
          (Audio|Parameter|SetBooleanParameter (Variables|FortGameplayCueNotifyLoop_Swinging|GetLayeredAudioComp) "IsMaxVelocity" (Utilities|Select :Index (Variables|/DryBox/Items/NyxGlass/B_Weapon_NyxGlass.B_Weapon_NyxGlass_C|GetState _nyxglassreference) :NewEnumerator0 false :NewEnumerator2 false :NewEnumerator7 false :NewEnumerator5 false :NewEnumerator6 true :NewEnumerator9 false) :!class /Script/AudioExtensions.AudioParameterControllerInterface)
          (bind _returnvalue (Utilities|Time|SetTimerbyEvent (EventDispatchers|CreateEvent :!event Custom|SetSpeedMultiplier) 0.1 true))
          (Variables|Default|SetSetFloatParamTimerHandle _returnvalue))
        (:else)))
    (:else)))

(event Custom|SetSpeedMultiplier
  (bind _layeredaudiocomp (Variables|FortGameplayCueNotifyLoop_Swinging|GetLayeredAudioComp))
  (bind _nyxglassreference (Variables|Default|GetNyxGlassReference)
    (:then
      (if (Math|Float|NearlyEqual(Float) (Variables|Default|GetSpeedMultiplier) (Utilities|RemapFloat (Math|Vector|VectorLength (Class|BWeaponGrappleParent|GetTrueVelocityBasedonTransform _nyxglassreference)) "(X=1500.000000,Y=3000.000000)" "(X=0.800000,Y=1.200000)" true) 0.05)
        (Utilities|Time|ClearandInvalidateTimerbyHandle _layeredaudiocomp (Variables|Default|GetSetFloatParamTimerHandle))
        (else
          (bind _output_get (Variables|Default|SetSpeedMultiplier (Math|Interpolation|FInterpTo (Variables|Default|GetSpeedMultiplier) (Utilities|RemapFloat (Math|Vector|VectorLength (Class|BWeaponGrappleParent|GetTrueVelocityBasedonTransform _nyxglassreference)) "(X=1500.000000,Y=3000.000000)" "(X=0.800000,Y=1.200000)" true) 0.1 2.0)))
          (Audio|Parameter|SetFloatParameter _layeredaudiocomp "SpeedMultiplier" _output_get :!class /Script/AudioExtensions.AudioParameterControllerInterface))))
    (:else)))

(fn AudioInit
  (params (GC%20Target :type Actor%20Object%20Reference))
  (bind _asplayer_pawn_athena (Utilities|Casting|CastToPlayerPawn_Athena GC%20Target))
  (bind _output_get (Variables|FortGameplayCueNotifyLoop_Swinging|SetPlayerPawn _asplayer_pawn_athena))
  (Transformation|AttachComponentToComponent (Variables|FortGameplayCueNotifyLoop_Swinging|GetLayeredAudioComp) (Variables|Transformation|GetRootComponent _output_get) "None" "SnapToTarget" "SnapToTarget" "SnapToTarget" true)
  (Variables|FortGameplayCueNotifyLoop_Swinging|SetbCanUpdateParameters true)
  (Collision|BindEventtoOnActorHit (Variables|FortGameplayCueNotifyLoop_Swinging|GetPlayerPawn) (EventDispatchers|CreateEvent :!event OnTargetPawnActorHit))
  (Audio|Parameter|SetBooleanParameter (Variables|FortGameplayCueNotifyLoop_Swinging|GetLayeredAudioComp) "IsLocal" (Pawn|IsLocallyViewed (Variables|FortGameplayCueNotifyLoop_Swinging|GetPlayerPawn)) :!class /Script/AudioExtensions.AudioParameterControllerInterface)
  (Audio|Parameter|SetBooleanParameter (Variables|FortGameplayCueNotifyLoop_Swinging|GetLayeredAudioComp) "IsTeam" (Utilities|Enum|Equal(Enum) (Team|GetTeamAffiliationforActortoLocalPlayer (Variables|FortGameplayCueNotifyLoop_Swinging|GetPlayerPawn)) "Friendly") :!class /Script/AudioExtensions.AudioParameterControllerInterface))

(fn AudioUninit
  (Audio|Components|Audio|FadeOut (Variables|FortGameplayCueNotifyLoop_Swinging|GetLayeredAudioComp) 0.2 0.0 "Logarithmic" :!class /Script/Engine.AudioComponent)
  (Variables|FortGameplayCueNotifyLoop_Swinging|SetbCanUpdateParameters false)
  (Collision|UnbindEventfromOnActorHit (Variables|FortGameplayCueNotifyLoop_Swinging|GetPlayerPawn) (EventDispatchers|CreateEvent :!event OnTargetPawnActorHit))
  (Class|FortLayeredAudioComponent|ResetAllFloatParameters (Variables|FortGameplayCueNotifyLoop_Swinging|GetLayeredAudioComp)))

(fn OnTargetPawnActorHit
  (params (SelfActor :type Actor%20Object%20Reference) (OtherActor :type Actor%20Object%20Reference) (NormalImpulse :type Vector) (Hit :type Hit%20Result%20Structure%20%28by%20ref%29))
  (if (Swinging|GetIsSwingAttached (Variables|FortGameplayCueNotifyLoop_Swinging|GetPlayerPawn))
    (Audio|OnCollision (Variables|Components|GetFortCollisionAudio) (Transformation|GetActorLocation SelfActor) (Transformation|GetVelocity (Variables|FortGameplayCueNotifyLoop_Swinging|GetPlayerPawn)))))

(fn ConstructionScript
  (|CallParentFunction))
