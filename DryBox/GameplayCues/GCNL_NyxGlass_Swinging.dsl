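; Looping gameplay cue for NyxGlass swinging that adds a post-process effect for the swinging player.
; On start, a timeline fades in the post-process blend weight if the pawn is locally controlled, and the second particle component is deactivated for local players.
; A repeating timer polls character movement and reverses the timeline while the local pawn moves on the ground.
; Removal clears the timer and zeroes the blend weight, leaving the component enabled.
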(event EventOnLoopingStart(EitherVFX-preferthisversion)
  (params (MyTarget :type Actor%20Object%20Reference) (Parameters :type Gameplay%20Cue%20Parameters%20Structure%20%28by%20ref%29) (ParticleComponents :type Array%20of%20FXSystem%20Component%20Object%20References) (AudioComponents :type Array%20of%20Audio%20Component%20Object%20References))
  (Utilities|FlowControl|Sequence
    (:then_0
      (bind _asfort_player_pawn_athena (Utilities|Casting|CastToFortPlayerPawnAthena MyTarget))
      (Variables|Default|SetAs%20Fort%20Player%20Pawn%20Athena _asfort_player_pawn_athena)
      (Utilities|FlowControl|Sequence
        (:then_0
          (bind _returnvalue (Utilities|Time|SetTimerbyEvent (delegate Custom|CE_Check) 0.1 true))
          (Variables|Default|SetTimer _returnvalue))
        (:then_1
          (if (Pawn|IsLocallyControlled (Variables|Default|GetAs%20Fort%20Player%20Pawn%20Athena))
            (Variables|/Script/Engine.PostProcessComponent|SetbEnabled true (Variables|Default|GetPostProcess))
            (bind (_direction _lerp) (Timeline|Timeline PlayFromStart)
              (:Update
                (Variables|/Script/Engine.PostProcessComponent|SetBlendWeight _lerp (Variables|Default|GetPostProcess))))))))
    (:then_1
      (if (Pawn|IsLocalPlayerController (Fortnite|GetFortPlayerControllerfromActor MyTarget))
        (Components|Activation|Deactivate (Utilities|Array|Get(acopy) ParticleComponents 1))))))

(event Custom|CE_Check
  (bind _as_fort_player_pawn_athena (Variables|Default|GetAs%20Fort%20Player%20Pawn%20Athena)
    (:then
      (if (AI|Components|NavMovement|IsMovingonGround (Variables|/Script/Engine.Character|GetCharacterMovement _as_fort_player_pawn_athena))
        (if (Pawn|IsLocallyControlled (Variables|Default|GetAs%20Fort%20Player%20Pawn%20Athena))
          (Timeline|Timeline Reverse))))
    (:else)))

(event EventOnRemoval(EitherVFX-preferthisversion)
  (params (MyTarget :type Actor%20Object%20Reference) (Parameters :type Gameplay%20Cue%20Parameters%20Structure%20%28by%20ref%29) (ParticleComponents :type Array%20of%20FXSystem%20Component%20Object%20References) (AudioComponents :type Array%20of%20Audio%20Component%20Object%20References) (BurstCameraShakeInstance :type Camera%20Shake%20Base%20Object%20Reference) (BurstDecalInstance :type Decal%20Actor%20Object%20Reference))
  (bind _postprocess (Variables|Default|GetPostProcess))
  (Utilities|Time|ClearandInvalidateTimerbyHandle MyTarget (Variables|Default|GetTimer))
  (Variables|/Script/Engine.PostProcessComponent|SetbEnabled true _postprocess)
  (Variables|/Script/Engine.PostProcessComponent|SetBlendWeight 0.0 _postprocess))

(fn ConstructionScript)
