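; Burst gameplay cue for the NyxGlass attack slash, playing its VFX and swing sound.
; Spawns a Niagara slash above the cue location, kept upright by flattening the effect causer's right vector, and feeds the target pawn's weapon mesh into two skeletal mesh parameters.
; Also spawns a radial force physics field there, attaches the sound to the weapon mesh and prints a debug string.
; The Niagara-only burst override is empty.
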(fn OnBurst(EitherVFX-preferthisversion)
  (params (MyTarget :type Actor%20Object%20Reference) (Parameters :type Gameplay%20Cue%20Parameters%20Structure%20%28by%20ref%29) (ParticleComponents :type Array%20of%20FXSystem%20Component%20Object%20References) (AudioComponents :type Array%20of%20Audio%20Component%20Object%20References) (BurstCameraShakeInstance :type Camera%20Shake%20Base%20Object%20Reference) (BurstDecalInstance :type Decal%20Actor%20Object%20Reference))
  (bind (_normalizedmagnitude _rawmagnitude _effectcontext _matchedtagname _originaltag _aggregatedsourcetags _aggregatedtargettags _location _normal _instigator _effectcauser) (Ability|GameplayCue|BreakGameplayCueParameters Parameters))
  (bind _returnvalue (Math|Vector|MakeVector 0.0 0.0 1.0))
  (bind _returnvalue_2 (Transformation|GetActorRightVector _effectcauser))
  (bind _returnvalue_3 (Math|Vector|ProjectVectorontoPlane _returnvalue_2 "0.000000,0.000000,1.000000"))
  (bind _returnvalue_4 (Math|Vector|Normalize _returnvalue_3))
  (bind _returnvalue_5 (Math|Rotator|MakeRotfromYZ _returnvalue_4 _returnvalue))
  (bind _returnvalue_6 (+ _location "0.000000,0.000000,60.000000"))
  (bind _asfort_player_pawn (Utilities|Casting|CastToFortPlayerPawn MyTarget))
  (bind _returnvalue_7 (Niagara|SpawnSystematLocation MyTarget "/DryBox/Items/NyxGlass/FX/Attack/NS_Slash.NS_Slash" _returnvalue_6 _returnvalue_5))
  (Niagara|SetNiagaraSkeletalMeshComponent _returnvalue_7 "SkeletalMesh" (Variables|/Script/FortniteGame.FortWeapon|GetWeaponMesh (Variables|/Script/FortniteGame.FortPawn|GetCurrentWeapon _asfort_player_pawn)))
  (Niagara|SetNiagaraSkeletalMeshComponent _returnvalue_7 "SkeletalMesh2" (Variables|/Script/FortniteGame.FortWeapon|GetWeaponMesh (Variables|/Script/FortniteGame.FortPawn|GetCurrentWeapon _asfort_player_pawn)))
  (Game|SpawnActorfromClass MyTarget "/Game/Effects/Niagara/Blueprints/PhsyicsFields/FN_RadialForce.FN_RadialForce_C" (Math|Transform|MakeTransform _returnvalue_6 _returnvalue_5) "AlwaysSpawn" "MultiplyWithRoot" 0 3.0 0.6 1.2e+03 1.5 0.0 0)
  (Audio|SpawnSoundAttached "/DryBox/Items/NyxGlass/Audio/NyxGlass_SwordAttack_Swing_Cue.NyxGlass_SwordAttack_Swing_Cue" (Variables|/Script/FortniteGame.FortWeapon|GetWeaponMesh (Variables|/Script/FortniteGame.FortPawn|GetCurrentWeapon _asfort_player_pawn)))
  (Development|PrintString MyTarget "ExecuteAttackSlashGC" true true "(R=1.000000,G=0.743303,B=0.000000,A=1.000000)")
  (return))

(fn OnBurst(Niagara)
  (params (MyTarget :type Actor%20Object%20Reference) (Parameters :type Gameplay%20Cue%20Parameters%20Structure%20%28by%20ref%29) (ParticleComponents :type Array%20of%20Niagara%20Particle%20System%20Component%20Object%20References) (AudioComponents :type Array%20of%20Audio%20Component%20Object%20References) (BurstCameraShakeInstance :type Camera%20Shake%20Base%20Object%20Reference) (BurstDecalInstance :type Decal%20Actor%20Object%20Reference))
  (return))
